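# Sound Cue for indoor spot ambience at the Kamp Krusty POI, playing one-shot creaks of a stressed wooden wall frame.
# A Random node picks among ten equally weighted creak waves, then a Modulator slightly varies the pitch.
# It uses a distant-creak attenuation and the indoor ambient one-shot sound class.
# A concurrency override caps simultaneous voices at two and sets a retrigger time, despite an assigned ambient spot concurrency set.
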SoundCue /WildEstate/Assets/Audio/POI/KampKrusty/Spots/SC_Spot_Wooden_Building_Stressing_Wall_Frame_Creaking_Movement_Int
root=Modulator_3
nodes [
 {id=Modulator_3, children [Random_1], properties {pitchMax=1.1, pitchMin=0.9}}
 {
  id=Random_1
  children [
   WavePlayer_5
   WavePlayer_9
   WavePlayer_10
   WavePlayer_11
   WavePlayer_12
   WavePlayer_13
   WavePlayer_14
   WavePlayer_15
   WavePlayer_16
   WavePlayer_17
  ]
  properties {weights [1,1,1,1,1,1,1,1,1,1]}
 }
 {
  id=WavePlayer_10
  properties {soundWaveAssetPtr=&/Spot_Wooden_Building_Stressing_Wall_Frame_Creaking_Movement_03}
 }
 {
  id=WavePlayer_11
  properties {soundWaveAssetPtr=&/Spot_Wooden_Building_Stressing_Wall_Frame_Creaking_Movement_04}
 }
 {
  id=WavePlayer_12
  properties {soundWaveAssetPtr=&/Spot_Wooden_Building_Stressing_Wall_Frame_Creaking_Movement_05}
 }
 {
  id=WavePlayer_13
  properties {soundWaveAssetPtr=&/Spot_Wooden_Building_Stressing_Wall_Frame_Creaking_Movement_06}
 }
 {
  id=WavePlayer_14
  properties {soundWaveAssetPtr=&/Spot_Wooden_Building_Stressing_Wall_Frame_Creaking_Movement_07}
 }
 {
  id=WavePlayer_15
  properties {soundWaveAssetPtr=&/Spot_Wooden_Building_Stressing_Wall_Frame_Creaking_Movement_08}
 }
 {
  id=WavePlayer_16
  properties {soundWaveAssetPtr=&/Spot_Wooden_Building_Stressing_Wall_Frame_Creaking_Movement_09}
 }
 {
  id=WavePlayer_17
  properties {soundWaveAssetPtr=&/Spot_Wooden_Building_Stressing_Wall_Frame_Creaking_Movement_10}
 }
 {
  id=WavePlayer_5
  properties {soundWaveAssetPtr=&/Spot_Wooden_Building_Stressing_Wall_Frame_Creaking_Movement_01}
 }
 {
  id=WavePlayer_9
  properties {soundWaveAssetPtr=&/Spot_Wooden_Building_Stressing_Wall_Frame_Creaking_Movement_02}
 }
]
properties {
 attenuationSettings=/Game/Sounds/Attenuation/Fort_TileAmbient_OneShot_Distant_Creak_Att
 bEnableBusSends=false
 bOverrideConcurrency=true
 maxDistance=8200
 soundClassObject=/Game/Sounds/FortSoundClassesAndModes/Fort_Ambient_OneShots_Indoors
 volumeMultiplier=1
 concurrencyOverrides {maxCount=2, retriggerTime=2, voiceStealReleaseTime=2, volumeScaleAttackTime=1}
 concurrencySet [/AmbientAudioBase/Concurrency/AmbientConc_3Max_Spot_Default_Concurrency]
 preEffectBusSends [{}]
}
