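# Sound cue for an indoor ambient spot one-shot of sustained wooden building creaking and structural stress at the Kamp Krusty POI.
# A Random node picks among nine equally weighted creak waves, and a root Modulator adds slight pitch variation to each play.
# It uses the indoor ambient one-shot sound class with distant-creak attenuation, and overrides concurrency to cap simultaneous voices, with a retrigger time and timed steal release.
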SoundCue /WildEstate/Assets/Audio/POI/KampKrusty/Spots/SC_Spot_Wooden_Building_Structural_Stress_Sustained_Creaking_Int
root=Modulator_3
nodes [
 {id=Modulator_3, children [Random_0], properties {pitchMax=1.1, pitchMin=0.9}}
 {
  id=Random_0
  children [WavePlayer_0, WavePlayer_1, WavePlayer_2, WavePlayer_3, WavePlayer_4, WavePlayer_6, WavePlayer_7, WavePlayer_8, WavePlayer_20]
  properties {weights [1,1,1,1,1,1,1,1,1]}
 }
 {
  id=WavePlayer_0
  properties {soundWaveAssetPtr=&/Spot_Wooden_Building_Structural_Stress_Sustained_Creaking_01}
 }
 {
  id=WavePlayer_1
  properties {soundWaveAssetPtr=&/Spot_Wooden_Building_Structural_Stress_Sustained_Creaking_02}
 }
 {
  id=WavePlayer_2
  properties {soundWaveAssetPtr=&/Spot_Wooden_Building_Structural_Stress_Sustained_Creaking_03}
 }
 {
  id=WavePlayer_20
  properties {soundWaveAssetPtr=&/Spot_Wooden_Building_Structural_Stress_Sustained_Creaking_09}
 }
 {
  id=WavePlayer_3
  properties {soundWaveAssetPtr=&/Spot_Wooden_Building_Structural_Stress_Sustained_Creaking_04}
 }
 {
  id=WavePlayer_4
  properties {soundWaveAssetPtr=&/Spot_Wooden_Building_Structural_Stress_Sustained_Creaking_05}
 }
 {
  id=WavePlayer_6
  properties {soundWaveAssetPtr=&/Spot_Wooden_Building_Structural_Stress_Sustained_Creaking_06}
 }
 {
  id=WavePlayer_7
  properties {soundWaveAssetPtr=&/Spot_Wooden_Building_Structural_Stress_Sustained_Creaking_07}
 }
 {
  id=WavePlayer_8
  properties {soundWaveAssetPtr=&/Spot_Wooden_Building_Structural_Stress_Sustained_Creaking_08}
 }
]
properties {
 attenuationSettings=/Game/Sounds/Attenuation/Fort_TileAmbient_OneShot_Distant_Creak_Att
 bEnableBusSends=false
 bOverrideConcurrency=true
 maxDistance=8200
 soundClassObject=/Game/Sounds/FortSoundClassesAndModes/Fort_Ambient_OneShots_Indoors
 volumeMultiplier=1
 concurrencyOverrides {maxCount=3, retriggerTime=2, voiceStealReleaseTime=2, volumeScaleAttackTime=1}
 concurrencySet [/AmbientAudioBase/Concurrency/AmbientConc_3Max_Spot_Default_Concurrency]
 preEffectBusSends [{}]
}
